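# EQS query that selects goals for ThawEpsonAI from its perceived enemies.
# A single composite option wraps an enemy generator, then scores each candidate by goal distance and perception.
# The distance test uses clamped inverse-linear scoring at half weight to favor closer enemies.
# The perception test weights candidates by sense type, with a small score for Athena hearing, and sets sense-age bounds.
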EnvQuery /ThawEpsonAI/Gameplay/EQS/EQS_GoalSelection_ThawEpsonAI
options [{
  generator {
   name=EnvQueryGenerator_Composite_1
   class=EnvQueryGenerator_Composite
   properties {
    optionName=Perceived Enemies
    generators [&self:EnvQueryGenerator_Composite_1.FortQueryGenerator_Enemies_1]
   }
  }
  tests [
   FortQueryTest_GoalDistance_3 {
    class=FortQueryTest_GoalDistance
    properties {
     bScoreEncounterGoals=true
     bScoreEnemies=true
     bScoreSpecificAssignments=false
     bScoreWorldGoals=false
     clampMaxType=SpecifiedValue
     clampMinType=SpecifiedValue
     filterType=Range
     floatValueMax=0
     scoreClampMax=5000
     scoreClampMin=200
     scoringEquation=InverseLinear
     scoringFactor=0.5
     testComment=Prefer closer things.
     testPurpose=Score
     assignmentIds []
    }
   }
   FortQueryTest_PerceptionAll_1 {
    class=FortQueryTest_PerceptionAll
    properties {
     clampMaxType=SpecifiedValue
     clampMinType=SpecifiedValue
     maxSenseAge=21
     minSenseAge=20
     scoreClampMax=1
     scoringFactor=1
     additionalSenseScores {&Class'&AthenaAISense_Hearing'=0.1}
     senseScores [0,0.1,1,0,0,0]
    }
   }
  ]
 }]
