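# EQS query for ThawEpsonAI wandering that picks a random, path-reachable point away from the querier.
# A coarse SimpleGrid spanning a wide area is projected onto navmesh with vertical tolerance and a small height offset.
# A 3D distance test filters out nearby points, a pathfinding test checks reachability, and a random test varies the choice.
# Projection and pathfinding share the Phoebe NoSmashable navigation filter.
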EnvQuery /ThawEpsonAI/Gameplay/EQS/EQS_ThawEpsonAI_Wander
options [{
  generator {
   name=EnvQueryGenerator_SimpleGrid_0
   class=EnvQueryGenerator_SimpleGrid
   properties {
    gridSize=5000
    spaceBetween=2000
    projectionData {
     extentX=10
     navigationFilter=&NavFilter_Phoebe_NoSmashable_C
     postProjectionVerticalOffset=30
     projectDown=5000
     projectUp=5000
    }
   }
  }
  tests [
   EnvQueryTest_Distance_0 {
    class=EnvQueryTest_Distance
    properties {
     distanceTo=&EnvQueryContext_Querier
     filterType=Minimum
     floatValueMin=1500
     scoringEquation=Linear
     scoringFactor=1
     testComment=""
     testMode=Distance3D
     testPurpose=Filter
    }
   }
   EnvQueryTest_Pathfinding_0 {
    class=EnvQueryTest_Pathfinding
    properties {filterClass=&NavFilter_Phoebe_NoSmashable_C}
   }
   FortQueryTest_Random_1 {class=FortQueryTest_Random, properties {floatValueMax=1, scoringFactor=1, testPurpose=FilterAndScore}}
  ]
 }]
