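# EQS query for ThawEpson AI that picks a reachable, nav-projected destination away from the nearest player.
# Three fallback options run in order: a randomly oriented circle with box-trace visibility scoring, the same circle without the trace, then a spiral donut around the querier.
# All require pathfinding, reject points too close or nearly overhead, and score higher with distance from the nearest player up to a clamp.
# The visibility trace only scores, avoiding bad interactions with vehicles.
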EnvQuery /ThawEpsonAI/Gameplay/EQS/EQS_ThawEpson_AwayFromPlayers
options [
 {
  generator {
   name=EnvQueryGenerator_OnCircle_5
   class=EnvQueryGenerator_OnCircle
   properties {
    bDefineArc=true
    circleRadius=5550
    numberOfPoints=6
    arcDirection {lineTo=/Game/Athena/AI/EQS/Contexts/EnvQueryContext_Querier_RandomPointOnUnitCircle_2D.EnvQueryContext_Querier_RandomPointOnUnitCircle_2D_C}
    projectionData {
     extentX=10
     navigationFilter=&NavFilter_Phoebe_NoSmashable_C
     postProjectionVerticalOffset=20
     projectDown=1500
     projectUp=1500
    }
    traceData {navigationFilter=null}
   }
  }
  tests [
   EnvQueryTest_Pathfinding_0 {class=EnvQueryTest_Pathfinding, properties {filterClass=null}}
   EnvQueryTest_Distance_6 {
    class=EnvQueryTest_Distance
    properties {
     clampMinType=SpecifiedValue
     distanceTo=&EnvQueryContext_Querier
     filterType=Minimum
     floatValueMin=1500
     scoreClampMin=250
     scoringEquation=Constant
     scoringFactor=1
     testComment=de-prioritize points we are very close to
     testMode=Distance3D
     testPurpose=FilterAndScore
    }
   }
   EnvQueryTest_Distance_7 {
    class=EnvQueryTest_Distance
    properties {
     clampMaxType=SpecifiedValue
     distanceTo=/Game/AI/EQS/Contexts/NearestPlayerContext.NearestPlayerContext_C
     filterType=Range
     floatValueMin=0
     scoreClampMax=6000
     scoringEquation=Linear
     scoringFactor=3
     testComment="Pick point farthest from NearestPlayer (unless NearestPlayer is SUPER far)"
     testMode=Distance3D
     testPurpose=Score
    }
   }
   EnvQueryTest_Trace_2 {
    class=EnvQueryTest_Trace
    properties {
     scoringFactor=10
     testComment="PREFER things with a clear visibility sweep, but don't require it. Visibility blocks on Vehicles, for instances, the FCC_Trace_World below does NOT. Fixes a lot of weird interactions with vehicles."
     testPurpose=Score
     traceData {extentX=0, extentY=20, extentZ=20, traceShape=Box}
    }
   }
   EnvQueryTest_Distance_8 {
    class=EnvQueryTest_Distance
    properties {
     distanceTo=&EnvQueryContext_Querier
     filterType=Minimum
     floatValueMin=1000
     scoringEquation=Constant
     scoringFactor=1
     testComment=Throw out points that are more or less directly overhead. We don't want to fly STRAIGHT UP.
     testMode=Distance2D
     testPurpose=Filter
    }
   }
  ]
 }
 {
  generator {
   name=EnvQueryGenerator_OnCircle_0
   class=EnvQueryGenerator_OnCircle
   properties {
    bDefineArc=true
    circleRadius=5550
    numberOfPoints=6
    arcDirection {lineTo=&EnvQueryContext_Querier_RandomPointOnUnitCircle_2D_C}
    projectionData {
     extentX=10
     navigationFilter=&NavFilter_Phoebe_NoSmashable_C
     postProjectionVerticalOffset=20
     projectDown=1500
     projectUp=1500
    }
    traceData {navigationFilter=null}
   }
  }
  tests [
   EnvQueryTest_Pathfinding_3 {class=EnvQueryTest_Pathfinding, properties {filterClass=null}}
   EnvQueryTest_Distance_0 {
    class=EnvQueryTest_Distance
    properties {
     clampMinType=SpecifiedValue
     distanceTo=&EnvQueryContext_Querier
     filterType=Minimum
     floatValueMin=1500
     scoreClampMin=250
     scoringEquation=Constant
     scoringFactor=1
     testComment=de-prioritize points we are very close to
     testMode=Distance3D
     testPurpose=FilterAndScore
    }
   }
   EnvQueryTest_Distance_1 {
    class=EnvQueryTest_Distance
    properties {
     clampMaxType=SpecifiedValue
     distanceTo=&NearestPlayerContext_C
     filterType=Range
     floatValueMin=0
     scoreClampMax=6000
     scoringEquation=Linear
     scoringFactor=3
     testComment="Pick point farthest from NearestPlayer (unless NearestPlayer is SUPER far)"
     testMode=Distance3D
     testPurpose=Score
    }
   }
   EnvQueryTest_Distance_2 {
    class=EnvQueryTest_Distance
    properties {
     distanceTo=&EnvQueryContext_Querier
     filterType=Minimum
     floatValueMin=1000
     scoringEquation=Constant
     scoringFactor=1
     testComment=Throw out points that are more or less directly overhead. We don't want to fly STRAIGHT UP.
     testMode=Distance2D
     testPurpose=Filter
    }
   }
  ]
 }
 {
  generator {
   name=EnvQueryGenerator_Donut_0
   class=EnvQueryGenerator_Donut
   properties {
    bUseSpiralPattern=true
    center=&EnvQueryContext_Querier
    innerRadius=2000
    numberOfRings=3
    outerRadius=5500
    pointsPerRing=6
    projectionData {
     extentX=10
     navigationFilter=&NavFilter_Phoebe_NoSmashable_C
     postProjectionVerticalOffset=20
     projectDown=1500
     projectUp=1500
    }
   }
  }
  tests [
   EnvQueryTest_Pathfinding_2 {class=EnvQueryTest_Pathfinding, properties {filterClass=null}}
   EnvQueryTest_Distance_9 {
    class=EnvQueryTest_Distance
    properties {
     clampMinType=SpecifiedValue
     distanceTo=&EnvQueryContext_Querier
     filterType=Minimum
     floatValueMin=1500
     scoreClampMin=250
     scoringEquation=Constant
     scoringFactor=1
     testComment=de-prioritize points we are very close to
     testMode=Distance3D
     testPurpose=FilterAndScore
    }
   }
   EnvQueryTest_Distance_10 {
    class=EnvQueryTest_Distance
    properties {
     clampMaxType=SpecifiedValue
     distanceTo=&NearestPlayerContext_C
     filterType=Range
     floatValueMin=0
     scoreClampMax=6000
     scoringEquation=Linear
     scoringFactor=3
     testComment="Pick point farthest from NearestPlayer (unless NearestPlayer is SUPER far)"
     testMode=Distance3D
     testPurpose=Score
    }
   }
   EnvQueryTest_Distance_11 {
    class=EnvQueryTest_Distance
    properties {
     distanceTo=&EnvQueryContext_Querier
     filterType=Minimum
     floatValueMin=1000
     scoringEquation=Constant
     scoringFactor=1
     testComment=Throw out points that are more or less directly overhead. We don't want to fly STRAIGHT UP.
     testMode=Distance2D
     testPurpose=Filter
    }
   }
  ]
 }
]
